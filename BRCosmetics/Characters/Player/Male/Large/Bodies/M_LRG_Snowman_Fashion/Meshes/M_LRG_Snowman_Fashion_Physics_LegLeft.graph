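# Physics asset for the left leg of the large male Snowman Fashion cosmetic body, covering a candy cane holster and its two candy canes.
# Each of the three dyn_ bones has a single box body.
# Both cane bodies are constrained to the holster body with equally tight swing and twist limit angles, forming a small branched rig with the holster as the shared hub.
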PhysicsAsset /BRCosmetics/Characters/Player/Male/Large/Bodies/M_LRG_Snowman_Fashion/Meshes/M_LRG_Snowman_Fashion_Physics_LegLeft
bodies {
 dyn_candyCane1 [{shapeType=Box, extentX=20.9081, extentY=4.9736, extentZ=4.8236, center=<-1.994 -1.4887 -1.4144>}]
 dyn_candyCane2 [{shapeType=Box, extentX=20.9181, extentY=5.8203, extentZ=2.7851, center=<-2.0009 -1.9249 0.4267>}]
 dyn_candyHolster [{shapeType=Box, extentX=22.0833, extentY=11.0919, extentZ=6.9228, center=<1.4309 0.2775 2.9825>}]
}
constraints [
 {bone1Name=dyn_candyCane1, bone2Name=dyn_candyHolster, swing1LimitDegrees=5, swing2LimitDegrees=5, twistLimitDegrees=5}
 {bone1Name=dyn_candyCane2, bone2Name=dyn_candyHolster, swing1LimitDegrees=5, swing2LimitDegrees=5, twistLimitDegrees=5}
]
