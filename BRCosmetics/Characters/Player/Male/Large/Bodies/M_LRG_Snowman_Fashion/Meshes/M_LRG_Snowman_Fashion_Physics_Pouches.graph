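# Physics asset for the three dynamic pouches on the large male Snowman Fashion body cosmetic.
# Each pouch bone has a box body, paired with a small box body on its own "_mstr" bone.
# A constraint joins each pair with tight angular limits, allowing more movement on the first swing axis than on the second swing or twist.
# The three pouch pairs share the same limits and have no constraints between them.
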PhysicsAsset /BRCosmetics/Characters/Player/Male/Large/Bodies/M_LRG_Snowman_Fashion/Meshes/M_LRG_Snowman_Fashion_Physics_Pouches
bodies {
 dyn_pouch1 [{shapeType=Box, extentX=13.3415, extentY=6.7307, extentZ=4.8558, center=<0.2703 0.0044 2.4039>}]
 dyn_pouch1_mstr [{shapeType=Box, extentX=1.01, extentY=1.01, extentZ=1.01}]
 dyn_pouch2 [{shapeType=Box, extentX=13.3415, extentY=6.7301, extentZ=4.8134, center=<0.2703 0.0122 2.3829>}]
 dyn_pouch2_mstr [{shapeType=Box, extentX=1.01, extentY=1.01, extentZ=1.01}]
 dyn_pouch3 [{shapeType=Box, extentX=13.3415, extentY=6.749, extentZ=4.8674, center=<0.2703 -0.0263 2.4096>}]
 dyn_pouch3_mstr [{shapeType=Box, extentX=1.01, extentY=1.01, extentZ=1.01}]
}
constraints [
 {bone1Name=dyn_pouch1, bone2Name=dyn_pouch1_mstr, swing1LimitDegrees=15, swing2LimitDegrees=5, twistLimitDegrees=5}
 {bone1Name=dyn_pouch2, bone2Name=dyn_pouch2_mstr, swing1LimitDegrees=15, swing2LimitDegrees=5, twistLimitDegrees=5}
 {bone1Name=dyn_pouch3, bone2Name=dyn_pouch3_mstr, swing1LimitDegrees=15, swing2LimitDegrees=5, twistLimitDegrees=5}
]
