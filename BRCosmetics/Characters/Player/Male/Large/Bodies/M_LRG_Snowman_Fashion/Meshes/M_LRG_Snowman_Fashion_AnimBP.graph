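# Post-copy anim blueprint for the large male Snowman Fashion body cosmetic, adding secondary motion to the pose copied from its attached parent mesh.
# Weighted multi-bone constraints place three pouch master bones across spine, clavicles and pelvis, then RigidBody nodes simulate the pouches relative to spine_01 and the left-leg physics asset relative to thigh_l.
# Each update derives back and forward alphas from the head's reference-pose delta against neck_02, blending additive collar_01 rotations.
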AnimBlueprint /BRCosmetics/Characters/Player/Male/Large/Bodies/M_LRG_Snowman_Fashion/Meshes/M_LRG_Snowman_Fashion_AnimBP
target_skeleton=&/M_LRG_Snowman_Fashion_Skeleton
graphs [
 AnimGraph {nodes [
   n1 = "Constraint - dyn_pouch1_mstr (PPPPPPP)"!(
    ComponentPose=Local To Component(LocalPose=Copy Pose From Mesh() {bUseAttachedParent=true})
    Alpha=1
    ConstraintWeights_0=0.0081
    ConstraintWeights_1:float=0.0628
    ConstraintWeights_2:float=0.1579
    ConstraintWeights_3:float=0.2366
    ConstraintWeights_4:float=0.2207
    ConstraintWeights_5:float=0.2924
    ConstraintWeights_6:float=0.0213
   ) -> (Pose) {
    alphaBoolBlend {blendOption=Linear}
    boneToModify {boneName=dyn_pouch1_mstr}
    constraintSetup [
     {targetBone {boneName=spine_01}}
     {targetBone {boneName=spine_02}}
     {targetBone {boneName=spine_03}}
     {targetBone {boneName=spine_04}}
     {targetBone {boneName=spine_05}}
     {targetBone {boneName=clavicle_r}}
     {targetBone {boneName=pelvis}}
    ]
    constraintWeights [1,1,1,1,1,1,1]
   }
   n2 = "Constraint - dyn_pouch2_mstr (PPPPPPPP)"!(
    ComponentPose=@n1
    Alpha=1
    ConstraintWeights_0=0.024
    ConstraintWeights_1:float=0.0322
    ConstraintWeights_2:float=0.1392
    ConstraintWeights_3:float=0.207
    ConstraintWeights_4:float=0.2563
    ConstraintWeights_5:float=0.1303
    ConstraintWeights_6:float=0.1474
    ConstraintWeights_7:float=0.0634
   ) -> (Pose) {
    alphaBoolBlend {blendOption=Linear}
    boneToModify {boneName=dyn_pouch2_mstr}
    constraintSetup [
     {targetBone {boneName=clavicle_l}}
     {targetBone {boneName=spine_01}}
     {targetBone {boneName=spine_02}}
     {targetBone {boneName=spine_03}}
     {targetBone {boneName=spine_04}}
     {targetBone {boneName=spine_05}}
     {targetBone {boneName=clavicle_r}}
     {targetBone {boneName=pelvis}}
    ]
    constraintWeights [1,1,1,1,1,1,1,1]
   }
   n3 = "Constraint - dyn_pouch3_mstr (PPPPPPPP)"!(
    ComponentPose=@n2
    Alpha=1
    ConstraintWeights_0=0.0428
    ConstraintWeights_1:float=0.0628
    ConstraintWeights_2:float=0.2269
    ConstraintWeights_3:float=0.2309
    ConstraintWeights_4:float=0.2092
    ConstraintWeights_5:float=0.0577
    ConstraintWeights_6:float=0.0413
    ConstraintWeights_7:float=0.1284
   ) -> (Pose) {
    alphaBoolBlend {blendOption=Linear}
    boneToModify {boneName=dyn_pouch3_mstr}
    constraintSetup [
     {targetBone {boneName=clavicle_l}}
     {targetBone {boneName=spine_01}}
     {targetBone {boneName=spine_02}}
     {targetBone {boneName=spine_03}}
     {targetBone {boneName=spine_04}}
     {targetBone {boneName=spine_05}}
     {targetBone {boneName=clavicle_r}}
     {targetBone {boneName=pelvis}}
    ]
    constraintWeights [1,1,1,1,1,1,1,1]
   }
   n4 = RigidBody(ComponentPose=@n3) {
    bClampLinearTranslationLimitToRefPose=true
    bOverrideWorldGravity=true
    overridePhysicsAsset=&/M_LRG_Snowman_Fashion_Physics_Pouches
    simulationSpace=BaseBoneSpace
    alphaBoolBlend {blendOption=Linear}
    baseBoneRef {boneName=spine_01}
    componentAppliedLinearAccClamp=<5000 5000 5000>
    componentLinearAccScale=<0.4 0.4 0.4>
    componentLinearVelScale=<0.3 0.3 0.3>
    overrideWorldGravity {z=-980}
   }
   n5 = RigidBody(ComponentPose=@n4) {
    bClampLinearTranslationLimitToRefPose=true
    bOverrideWorldGravity=true
    overridePhysicsAsset=&/M_LRG_Snowman_Fashion_Physics_LegLeft
    simulationSpace=BaseBoneSpace
    alphaBoolBlend {blendOption=Linear}
    baseBoneRef {boneName=thigh_l}
    componentAppliedLinearAccClamp=<5000 5000 5000>
    componentLinearAccScale=<0.4 0.4 0.4>
    componentLinearVelScale=<0.3 0.3 0.3>
    overrideWorldGravity {z=-980}
   }
   n6 = "Transform (Modify) Bone - Bone: collar_01"(ComponentPose=@n5, Alpha=Get headBackAlpha() -> (headBackAlpha:double), Translation="0,0,0", Rotation="10.000,-20.000,0.000", Scale="1,1,1") {
    rotationMode=BMM_Additive
    rotationSpace=BCS_BoneSpace
    scaleMode=BMM_Additive
    scaleSpace=BCS_BoneSpace
    translationMode=BMM_Additive
    translationSpace=BCS_BoneSpace
    alphaBoolBlend {blendOption=Linear}
    boneToModify {boneName=collar_01}
    rotation {pitch=10, yaw=-20}
   }
   n7 = "Transform (Modify) Bone - Bone: collar_01"(ComponentPose=@n6, Alpha=Get headFwdAlpha() -> (headFwdAlpha:double), Translation="0,0,0", Rotation="0.000,25,0.000", Scale="1,1,1") {
    rotationMode=BMM_Additive
    rotationSpace=BCS_BoneSpace
    scaleMode=BMM_Additive
    scaleSpace=BCS_BoneSpace
    translationMode=BMM_Additive
    translationSpace=BCS_BoneSpace
    alphaBoolBlend {blendOption=Linear}
    boneToModify {boneName=collar_01}
    rotation {yaw=25}
   }
   Output Pose(Result=Component To Local(ComponentPose=@n7)) {layerGroup=None}
  ]}
 EventGraph {nodes [
   Event BlueprintUpdateAnimation()
   n1 = GetOwnerSkeletalMeshComponent()
   > Set headBackAlpha(headBackAlpha:double=GetAlphaFromReferenceTransformDelta(SkeletalMeshComponent=@n1, BoneName=head, BaseName=neck_02, InputRange=(X=-20,Y=-30), TwistAxis="1,0,0")) -> (Output_Get:double)
   > Set headFwdAlpha(headFwdAlpha:double=GetAlphaFromReferenceTransformDelta(SkeletalMeshComponent=@n1, BoneName=head, BaseName=neck_02, InputRange=(X=-1,Y=10), TwistAxis="1,0,0")) -> (Output_Get:double)
  ]}
]
variables [
 headBackAlpha {type=number}
 headFwdAlpha {type=number}
]
